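# MF_Artemis_SnowTops is a material function that outputs snow material attributes, named for use on snow tops.
# It takes the output of MF_Athena_Snow_01 and overrides selected attributes, setting emissive and subsurface color from a default constant.
# A UseColorMap static switch, on by default, runs base color through MF_Athena_GrassMatch, otherwise passing the snow color through unchanged.
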MaterialFunction /Game/Athena/Environments/Landscape/MaterialFunctions/Snow/MF_Artemis_SnowTops
graph {nodes [
  n1 = GetMaterialAttributes(None=MaterialFunctionCall!(UseLandscapeCoordsOnSwitch) {materialFunction=&/MF_Athena_Snow_01}.Material) {attributeGetTypes [#1, #2]}
  n2 = Constant()
  n3 = SetMaterialAttributes!(
   MaterialAttributes=@n1.MaterialAttributes
   Normal=@n1.MaterialAttributes
   EmissiveColor=@n2
   SubsurfaceColor=@n2
   BaseColor=StaticSwitchParameter(
    True=MaterialFunctionCall!(BaseColor=@n1.BaseColor) {materialFunction=/Game/Athena/Environments/Landscape/MaterialFunctions/Standard/MF_Athena_GrassMatch}.Result
    False=@n1.BaseColor
   ) {defaultValue=true, parameterName=UseColorMap}
  ) {attributeSetTypes [
    #1
    #3
    #4
    #2
   ]}
  FunctionOutput(None=@n3)
 ]}
