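# Material function for Season 7 Athena landscape snow materials that applies a test color tint to a base color.
# A topo mask's alpha, sampled by default with terrain world coordinates, scales by an MPC amount and clamps to drive the blend.
# The grass-masked result blends toward desaturated color times an MPC tint.
# ResultWithRock layers a rock-masked blend toward the ColorTest2 parameter tint over that result, which is also output separately.
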MaterialFunction /Game/Athena/Environments/Landscape/MaterialFunctions/Snow/MF_Athena_ColorTest_S07
graph {nodes [
  n1 = FunctionInput() {inputName=BaseColor}
  n2 = Desaturation(
   None=@n1
   Fraction=FunctionInput(Preview=Constant() {r=0.75}) {bUsePreviewValueAsDefault=true, inputName=DesaturationAmount, inputType=FunctionInput_Scalar}
  )
  n3 = CollectionParameter() {collection=/Game/Athena/Environments/Landscape/MPC/MPC_Landscape_S07, parameterName=ColorTest_Tint}
  n4 = TextureSampleParameter2D(UVs=FunctionInput(Preview=MaterialFunctionCall!(WorldPosition) {materialFunction=/Game/Athena/Environments/Landscape/MaterialFunctions/Utilities/MF_Athena_Terrain_Coords}.Result) {bUsePreviewValueAsDefault=true, inputName=UVs, inputType=FunctionInput_Vector2}) {
   group=Snow
   parameterName=TopoMask
   samplerType=SAMPLERTYPE_LinearColor
   texture=/Game/Athena/Environments/Landscape/Textures/T_Athena_Terrain_Topo_Mask
  }
  n5 = Clamp(None=Multiply(A=@n4.A, B=CollectionParameter() {collection=&MPC_Landscape_S07, parameterName=ColorTestAmount}))
  n6 = Constant() {r=1}
  n7 = LinearInterpolate(
   A=@n1
   B=Multiply(A=@n2, B=ComponentMask(None=@n3) {b=true, g=true, r=true})
   Alpha=Multiply(A=@n5, B=FunctionInput(Preview=@n6) {bUsePreviewValueAsDefault=true, inputName=GrassMask, inputType=FunctionInput_Scalar})
  )
  n8 = VectorParameter() {parameterName=ColorTest2, defaultValue=<rgba 0.6198 0.4639 0.2954 0>}
  n9 = FunctionInput(Preview=Constant() {r=1}) {bUsePreviewValueAsDefault=true, inputName=RockMask, inputType=FunctionInput_Scalar}
  FunctionOutput(None=LinearInterpolate(A=FunctionOutput(None=@n7), B=Multiply(A=@n2, B=@n8.RGB), Alpha=Multiply(A=@n5, B=@n9))) {outputName=ResultWithRock}
 ]}
